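# Default sound submix for the Resonance Audio plugin, serving as a root submix with no parent or child submixes.
# Its effect chain holds a single effect preset, ResonanceReverbDefault, so audio routed into this submix passes through that default Resonance reverb.
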SoundSubmix /ResonanceAudio/ResonanceSubmixDefault
graph_type=sound_submix
parent=null
children []
properties {submixEffectChain [&/ResonanceReverbDefault]}
