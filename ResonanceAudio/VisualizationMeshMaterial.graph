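# Translucent, two-sided material in the Resonance Audio plugin, named for use on a visualization mesh.
# It uses the Default Lit shading model with decal response turned off.
# Its graph holds only two parameters, a scalar Opacity defaulting to half and a cyan color vector, with no textures or math nodes.
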Material /ResonanceAudio/VisualizationMeshMaterial
graph {
 properties {
  bUseLightmapDirectionality=false
  bUseTranslucencyVertexFog=false
  blendMode=BLEND_Translucent
  materialDecalResponse=MDR_None
  pixelDepthOffsetMode=PDOM_Legacy
  twoSided=true
  usedShadingModels=Default Lit
 }
 nodes [
  ScalarParameter() {defaultValue=0.5, parameterName=Opacity}
  VectorParameter() {defaultValue=<rgba 0 198 246 0>}
 ]
}
